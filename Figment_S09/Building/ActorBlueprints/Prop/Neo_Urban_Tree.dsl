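; Neo Urban tree prop Blueprint that optionally adds a falling or drifting leaf particle effect to the tree.
; On BeginPlay, if Use_BR_Leaves is set, it spawns the LeafEmitter attached to the static mesh component at the OffsetHeight offset.
; The spawned emitter takes its SpawnRateScale parameter from a Blueprint variable and gets a capped max draw distance to limit cost.
; The construction script is empty.
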(event EventBeginPlay
  (if (Variables|BR_Effects|GetUse_BR_Leaves)
    (bind _returnvalue (Effects|SpawnEmitterAttached :EmitterTemplate (Variables|Default|GetLeafEmitter) :AttachToComponent (Variables|Debug|GetStaticMeshComponent) :AttachPointName "None" :Location (Variables|BR_Effects|GetOffsetHeight) :Scale "1.000000,1.000000,1.000000" :LocationType "SnapToTarget" :bAutoDestroy true :PoolingMethod "None" :bAutoActivate true))
    (Effects|Components|ParticleSystem|SetFloatParameter _returnvalue "SpawnRateScale" (Variables|BR_Effects|GetSpawnRateScale))
    (LOD|SetMaxDrawDistance _returnvalue 4e+03)))

(fn ConstructionScript)
