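; Actor Blueprint for a Dumpling mascot sign prop whose mascot mesh rotates on non-mobile platforms.
; On BeginPlay it checks for a mobile platform and deactivates the RotatingMovement component there.
; Otherwise it sets the S_Dumpling_Mascot component as the rotating movement's updated component, so only that part spins.
; The construction script only calls its parent.
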(event EventBeginPlay
  (bind _rotatingmovement (Variables|Default|GetRotatingMovement))
  (if (Fortnite|IsMobilePlatform)
    (Components|Activation|Deactivate _rotatingmovement)
    (else
      (Components|Movement|SetUpdatedComponent _rotatingmovement (Variables|Default|GetS_Dumpling_Mascot)))))

(fn ConstructionScript
  (|CallParentFunction))
